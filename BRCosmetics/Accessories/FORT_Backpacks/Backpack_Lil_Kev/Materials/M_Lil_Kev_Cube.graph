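# Lit material for the Lil Kev cube back bling, with textured base color, normal, specular, metallic and roughness, plus animated emissive.
# Time-driven sines, phase-offset by an FX mask, flicker a gradient panning along pre-skinned Z, tinted by two color parameters.
# A pre-skinned-space SphereMask, shifted by normals and PivotOffset, adds spec-weighted edge-color glow.
# FX blue masks the emissive; attributes pass through the accessories TechArtOverrides function.
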Material /BRCosmetics/Accessories/FORT_Backpacks/Backpack_Lil_Kev/Materials/M_Lil_Kev_Cube
graph {
 properties {
  bUseMaterialAttributes=true
  bUsedWithMeshDeformer=true
  bUsedWithSkeletalMesh=true
  pixelDepthOffsetMode=PDOM_Legacy
  usedShadingModels=Default Lit
 }
 nodes [
  n1 = TextureSampleParameter2D() {
   parameterName=Diffuse
   samplerSource=SSM_Wrap_WorldGroupSettings
   texture=/BRCosmetics/Accessories/FORT_Backpacks/Backpack_Lil_Kev/Textures/T_Lil_Kev_D
  }
  n2 = TextureSampleParameter2D() {
   desc="R = SpecMask G = Metallic B = Roughness"
   group=Specular
   parameterName=SpecularMasks
   samplerSource=SSM_Wrap_WorldGroupSettings
   samplerType=SAMPLERTYPE_Masks
   texture=/BRCosmetics/Accessories/FORT_Backpacks/Backpack_Lil_Kev/Textures/T_Lil_Kev_S
  }
  n3 = TextureSampleParameter2D() {
   group=Emissive
   parameterName=EmissiveMasks
   samplerSource=SSM_Wrap_WorldGroupSettings
   samplerType=SAMPLERTYPE_Masks
   texture=/BRCosmetics/Accessories/FORT_Backpacks/Backpack_Lil_Kev/Textures/T_Lil_Kev_FX
  }
  n4 = LinearInterpolate(Alpha=@n3.R) {constA=1, constB=0.5}
  n5 = Add(A=Time(), B=@n3.G)
  n6 = Sine(None=@n5) {period=0.22}
  n7 = Sine(None=@n5) {period=0.11}
  n8 = Clamp(None=Add(A=Multiply(A=@n6, B=@n7), B=Sine(None=@n5) {period=0.65})) {maxDefault=2}
  n9 = ConstantBiasScale(None=Sine(None=@n5) {period=3})
  n10 = LinearInterpolate(A=Multiply(A=Multiply(A=@n8) {constB=2}, B=@n9), Alpha=Power(Base=@n9) {constExponent=3})
  n11 = ComponentMask(None=VertexInterpolator(VS=PreSkinnedPosition()).PS) {b=true}
  n12 = Constant() {desc=gradient tile, r=50}
  n13 = Constant() {desc=gradient pan speed, r=-0.27}
  n14 = Add(A=Time())
  n15 = TextureSample(UVs=Add(A=Divide(A=@n11, B=@n12) {constB=200}, B=Multiply(A=@n13, B=@n14) {constB=-0.2})) {texture=/Game/Global/Textures/Default/Masks_and_Gradients/T_VertGradient_02_sRGB}
  n16 = VectorParameter() {group=Emissive, parameterName=EmissiveEdgeColor, defaultValue=<rgba 0.3581 0.0573 1 0>}
  n17 = LinearInterpolate(A=@n16.RGB, B=VectorParameter() {group=Emissive, parameterName=EmissiveColor, defaultValue=<rgba 0.4924 0.0624 0.5208 0>}.RGB, Alpha=@n3.R)
  n18 = Add(A=@n2.R, B=@n2.R)
  n19 = VertexInterpolator(VS=PreSkinnedPosition())
  n20 = Subtract(A=VertexInterpolator(VS=PreSkinnedNormal()).PS, B=VectorParameter() {parameterName=PivotOffset, defaultValue=<rgba 0 8.65 0 0>}.RGB)
  n21 = ScalarParameter() {defaultValue=18, parameterName=GlowRadius}
  n22 = ScalarParameter() {parameterName=GlowHardness}
  n23 = Multiply(
   A=LinearInterpolate(Alpha=Saturate(None=@n18)) {constA=2, constB=0.5}
   B=Power(
    Base=SphereMask(A=@n19.PS, B=@n20, Radius=@n21, Hardness=@n22) {hardnessPercent=30}
    Exp=ScalarParameter() {defaultValue=2.5, parameterName=GlowPower}
   )
  )
  n24 = ScalarParameter() {defaultValue=1, parameterName=Brightness}
  n25 = Add(A=Multiply(A=Multiply(A=@n10, B=@n15.G), B=Multiply(A=@n3.R, B=@n17)), B=Multiply(A=@n16.RGB, B=Multiply(A=@n23, B=@n24)))
  n26 = TextureSampleParameter2D() {parameterName=Normals, samplerType=SAMPLERTYPE_Normal, texture=/BRCosmetics/Accessories/FORT_Backpacks/Backpack_Lil_Kev/Textures/T_Lil_Kev_N}
  n27 = SetMaterialAttributes!(
   MaterialAttributes
   Base Color=NamedRerouteDeclaration(None=@n1.RGB) {name=baseColor, nodeColor=<rgba 1 0 0.2109 1>}
   Metallic=NamedRerouteDeclaration(None=@n2.G) {name=metallic, nodeColor=<rgba 0.125 1 0 1>}
   Specular=NamedRerouteDeclaration(None=Multiply(A=@n4, B=@n2.R)) {name=spec, nodeColor=<rgba 1 0 0.8672 1>}
   Roughness=NamedRerouteDeclaration(None=@n2.B) {name=roughness, nodeColor=<rgba 0.3125 0 1 1>}
   Emissive Color=NamedRerouteDeclaration(None=LinearInterpolate(B=@n25, Alpha=@n3.B)) {name=emissive, nodeColor=<rgba 0 1 0.7422 1>}
   Normal=NamedRerouteDeclaration(None=@n26.RGB) {name=normal, nodeColor=<rgba 0 1 0.2734 1>}
  ) {attributeSetTypes [
    #1
    #2
    #3
    #4
    #5
    #6
   ]}
  MaterialFunctionCall(In=@n27) {materialFunction=&MF_TechArtOverrides_accessories}
 ]
}
