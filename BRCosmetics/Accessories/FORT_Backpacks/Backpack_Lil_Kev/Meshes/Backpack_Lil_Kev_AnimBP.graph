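# Animation Blueprint for the Lil Kev backpack cosmetic that adds physical jiggle to its two googly eye bones.
# The input pose passes through chained Anim Dynamics nodes for googly_eye and googly_eye_small, simulated in world space, then returns to local space for output.
# Each eye slides within small Y/Z linear limits with zero angular range, tighter for the small eye, plus a spherical limit.
# The event graph holds no logic.
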AnimBlueprint /BRCosmetics/Accessories/FORT_Backpacks/Backpack_Lil_Kev/Meshes/Backpack_Lil_Kev_AnimBP
target_skeleton=&/Backpack_Lil_Kev_Skeleton
graphs [
 AnimGraph {nodes [
   AnimGraphNode_SubInput_0 = Input Pose()
   n1 = Anim Dynamics - Bone: googly_eye(ComponentPose=Local To Component(LocalPose=@AnimGraphNode_SubInput_0), Alpha=1, ExternalForce="0,0,0") {
    bGravityOverrideInSimSpace=true
    simulationSpace=World
    alphaBoolBlend {blendOption=Linear}
    boundBone {boneName=googly_eye}
    physicsBodyDefinitions [{
      sphereCollisionRadius=0
      boundBone {boneName=googly_eye}
      boxExtents=<0.5 3 3>
      constraintSetup {angularLimitsMax=<0 0 0>, angularLimitsMin=<0 0 0>, angularTarget=<0 0 0>, linearAxesMax=<0 1.3 1.5>, linearAxesMin=<0 -1.3 -1.5>}
      localJointOffset=<0.05 0 0>
     }]
    sphericalLimits [{limitRadius=5, drivingBone {boneName=None}, sphereLocalOffset=<-5 0 0>}]
   }
   n2 = Anim Dynamics - Bone: googly_eye_small(ComponentPose=@n1, Alpha=1, ExternalForce="0,0,0") {
    bGravityOverrideInSimSpace=true
    simulationSpace=World
    alphaBoolBlend {blendOption=Linear}
    boundBone {boneName=googly_eye_small}
    physicsBodyDefinitions [{
      sphereCollisionRadius=0
      boundBone {boneName=googly_eye_small}
      boxExtents=<0.5 2 2>
      constraintSetup {angularLimitsMax=<0 0 0>, angularLimitsMin=<0 0 0>, angularTarget=<0 0 0>, linearAxesMax=<0 0.7 1.2>, linearAxesMin=<0 -0.7 -1.2>}
      localJointOffset=<-0.1 0 0>
     }]
    sphericalLimits [{limitRadius=5, drivingBone {boneName=None}, sphereLocalOffset=<-5 0 0>}]
   }
   Output Pose(Result=Component To Local(ComponentPose=@n2)) {layerGroup=None}
  ]}
 EventGraph {nodes [Event BlueprintUpdateAnimation()]}
]
variables []
